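# Post-process material that adds a centered overlay texture onto the camera image, driven by OverlaySize and OverlayOpacity from MPC_Drawing.
# Screen UVs scale around the center by OverlaySize; both aspect-ratio terms divide a component by itself, so no aspect correction occurs.
# The texture gets a 1/2.2 gamma power, is clipped by a hard-edged box mask, scaled by opacity and added to scene color.
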Material /Landmass/Landscape/OldPrototype_BP/Mats/PP/CameraOverlay
graph {
 properties {materialDomain=MD_PostProcess, pixelDepthOffsetMode=PDOM_Legacy, usedShadingModels=Default Lit}
 nodes [
  n1 = MaterialFunctionCall(Float2=ViewSize()) {materialFunction=&BreakOutFloat2Components}
  n2 = AppendVector(A=Divide(A=@n1.G, B=@n1.G), B=Constant() {r=1})
  n3 = CollectionParameter() {collection=/Landmass/Landscape/OldPrototype_BP/Mats/MPC/MPC_Drawing, parameterName=OverlaySize}
  n4 = MaterialFunctionCall(Float2=TextureProperty(Tex=TextureObjectParameter() {parameterName=OverlayTexture, texture=/Engine/EditorResources/BadBlueprintSprite})) {materialFunction=&BreakOutFloat2Components}
  n5 = Divide(A=@n4.R, B=@n4.R)
  n6 = Constant() {r=1}
  n7 = MaterialFunctionCall(UVs=ScreenPosition().ViewportUV, Texture Scale=Multiply(A=Multiply(A=@n2, B=@n3), B=AppendVector(A=@n5, B=@n6))) {materialFunction=&ScaleUVsByCenter}
  n8 = Power(Base=TextureSampleParameter2D(UVs=@n7.UVs) {parameterName=OverlayTexture, texture=&BadBlueprintSprite}.RGB) {constExponent=0.4545}
  n9 = MaterialFunctionCall!(A=@n7.UVs, B=Constant() {r=0.5}, Bounds=ScalarParameter() {defaultValue=1, parameterName=B}, Edge Falloff=Constant() {r=0.001}) {materialFunction=/Engine/Functions/Engine_MaterialFunctions01/Gradient/BoxMask-2D}
  Add(
   A=ComponentMask(None=SceneTexture() {sceneTextureId=PPI_PostProcessInput0}.Color) {b=true, g=true, r=true}
   B=Multiply(
    A=Multiply(A=@n8, B=@n9.Result)
    B=CollectionParameter() {collection=&MPC_Drawing, parameterName=OverlayOpacity}
   )
  )
 ]
}
